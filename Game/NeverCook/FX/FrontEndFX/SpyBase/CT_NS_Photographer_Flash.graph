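# Niagara camera-flash effect for a photographer, filed under front-end Spy Base FX.
# FLash bursts one short-lived, camera-offset particle to two sprite renderers, also writing custom bigger-size and brighter-color attributes from curves.
# FLash001 bursts a matching particle at a location offset to drive a light renderer, with its own emitter timing.
# Curves give both a quick rise and fade in size and alpha, with strongly boosted RGB.
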NiagaraSystem /Game/NeverCook/FX/FrontEndFX/SpyBase/CT_NS_Photographer_Flash
dependencies {
 usedDataInterfaces [&NiagaraDataInterfaceCurve, &NiagaraDataInterfaceColorCurve]
 usedDynamicInputs [&UniformRangedFloat]
 usedModules [
  &self:SystemSpawnScript.SystemScriptSource.SystemScriptGraph.NiagaraNodeAssignment_0.SetVariables_#1
  &SystemState
 ]
}
emitters {
 FLash {
  emitterUpdateScript [
   &EmitterState()
   &SpawnBurst_Instantaneous(Spawn Count=1, Use Spawn Probability=false)
  ]
  particleSpawnScript [
   /Niagara/Modules/Spawn/Initialization/InitializeParticle(
    Lifetime=0.2
    Module.Write Color=true
    Module.Write Lifetime=true
    Module.Write Mass=true
    Module.Write Material Random=true
    Module.Write Position=false
    Module.Write Scale=false
    Module.Write SpriteRotation=true
    Module.Write SpriteSize=true
    Position=&SimulationPosition()
    Sprite Size=<100 100>
   )
   &CameraOffset(Camera Offset Amount=15)
  ]
  particleUpdateScript [
   &ParticleState()
   &SolveForcesAndVelocity(Clamp Velocity=false, Limit Acceleration=false)
   &ScaleSpriteSize(Scale Factor=&Vector2DFromFloat(Value=&FloatFromCurve(FloatCurve=DataInterface() {propertyValues {Curve {keys [<at -0.0046 0>,<at 0.1423 0.9767>,<at 1.0011 0.002>]}}}, Scale Curve=0.2)))
   /SetVariables(Particles.Size_Bigger:Vector2f=&Vector2DFromFloat(Value=&FloatFromCurve(FloatCurve=DataInterface() {propertyValues {Curve {keys [<at -0.0046 0>,<at 0.5 1>,<at 1.0011 0.002>]}}}, Scale Curve=150))) {isSetParameters=true}
   /SetVariables(Particles.Color_Brighter:LinearColor=&ColorFromCurve(ColorCurve=DataInterface() {propertyValues {
       AlphaCurve {keys [<at 0.0072 0>, {arriveTangent=-0.003, interpMode=RCIM_Cubic, leaveTangent=-0.003, time=0.1402, value=1}, <at 0.9977 -0.003 RCIM_Cubic>]}
       BlueCurve {keys [<at 0.0013 25 RCIM_Cubic>]}
       GreenCurve {keys [<at 0.0013 25 RCIM_Cubic>]}
       RedCurve {keys [<at 0.0013 25 RCIM_Cubic>]}
      }})) {isSetParameters=true}
   &ScaleColor(
    Scale Alpha=&FloatFromCurve(FloatCurve=DataInterface() {propertyValues {Curve {keys [<at -0.0046 0>,<at 0.1241 0.9674>,<at 1.0011 0.002>]}}})
    Scale RGB=<25 25 25>
    ScaleA=true
    ScaleRGB=true
   )
  ]
  rendererClasses [&NiagaraSpriteRendererProperties]
  renderers [
   {rendererClass=&NiagaraSpriteRendererProperties}
   {rendererClass=&NiagaraSpriteRendererProperties, rendererIndex=1}
  ]
 }
 FLash001 {
  emitterUpdateScript [
   &EmitterState(Life Cycle Mode=::NewEnumerator1, Loop Behavior=::NewEnumerator1, Loop Delay=0.017, Loop Duration=1, UseLoopDelay=true)
   &SpawnBurst_Instantaneous(Spawn Count=1, Use Spawn Probability=false)
  ]
  particleSpawnScript [
   &InitializeParticle(
    Lifetime=0.2
    Module.Write Color=true
    Module.Write Lifetime=true
    Module.Write Mass=true
    Module.Write Material Random=true
    Module.Write Position=false
    Module.Write Scale=false
    Module.Write SpriteRotation=true
    Module.Write SpriteSize=true
    Position=&SimulationPosition()
    Sprite Size=<100 100>
   )
   &SystemLocation(MaskSpawnGroup=false, Offset=<35 0 0>)
  ]
  particleUpdateScript [
   &ParticleState()
   &SolveForcesAndVelocity(Clamp Velocity=false, Limit Acceleration=false)
   &ScaleSpriteSize(Scale Factor=&Vector2DFromFloat(Value=&FloatFromCurve(FloatCurve=DataInterface() {propertyValues {Curve {keys [<at -0.0046 0>,<at 0.1423 0.9767>,<at 1.0011 0.002>]}}}, Scale Curve=0.2)))
   &ScaleColor(
    Scale Alpha=&FloatFromCurve(FloatCurve=DataInterface() {propertyValues {Curve {keys [<at -0.0046 0>,<at 0.1241 0.9674>,<at 1.0011 0.002>]}}})
    Scale RGB=<25 25 25>
    ScaleA=true
    ScaleRGB=true
   )
  ]
  rendererClasses [&NiagaraLightRendererProperties]
  renderers [{rendererClass=&NiagaraLightRendererProperties, rendererIndex=0}]
 }
}
summary {
 systemName=CT_NS_Photographer_Flash
 emitters [
  {emitterName=FLash, rendererClasses [&NiagaraSpriteRendererProperties]}
  {emitterName=FLash001, rendererClasses [&NiagaraLightRendererProperties]}
 ]
 userVariables [
  User.GlobalOffset {type=Vector3f, description="", defaultValue {struct=Vector3f, value=<0 0 0>}}
  User.RibbonIDOffset {type=NiagaraInt32}
 ]
}
