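# Niagara effect for a flopper tank in the front-end SpyBase FX set, showing a single fish mesh circling in place.
# The Fish emitter burst-spawns one scaled-up mesh particle, and custom orbit modules seed and drive its slow circular motion around a slightly tilted axis.
# The particle is not killed when its lifetime elapses.
# Rotational velocity and initial mesh orientation modules are present but disabled.
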NiagaraSystem /Game/NeverCook/FX/FrontEndFX/SpyBase/NS_FlopperTank_02
dependencies {usedDataInterfaces [&NiagaraDataInterfaceCurve], usedDynamicInputs [], usedModules [&SystemState]}
emitters {Fish {
  emitterUpdateScript [
   &EmitterScalability(
    Module.bApplySpawnCountScale=false
    Module.bApplySpawnCountScaleByDistanceFraction=false
    Module.bUseMaxDistance=false
    Module.bUseMinDistance=false
   ) {enabled=false}
   &EmitterLifeCycle() {enabled=false}
   &EmitterState(
    Inactive Response=::NewEnumerator2
    Life Cycle Mode=::NewEnumerator1
    Loop Behavior=::NewEnumerator1
    Loop Duration Mode=::NewEnumerator1
    UseLoopDelay=false
   )
   &SpawnBurst_Instantaneous(Spawn Count=1, Use Spawn Probability=false)
  ]
  particleSpawnScript [
   /Game/Effects/Niagara/Modules/Orbit/FN_Orbit_Init_Variables_v2(Seed Offset=650)
   /Niagara/Modules/Spawn/Initialization/InitializeParticle(
    Mesh Scale=<3 3 3>
    Module.Write Color=true
    Module.Write Lifetime=true
    Module.Write Mass=true
    Module.Write Material Random=true
    Module.Write Position=false
    Module.Write Scale=true
    Module.Write SpriteRotation=false
    Module.Write SpriteSize=false
    Position=&SimulationPosition()
   )
   &AddRotationalVelocity(
    Coordinate Space:ENiagaraMeshTransforms=::NewEnumerator1
    Rotation Rate=&UniformRangedVector(Minimum=<-0.25 0 0>, Maximum=<0.25 0 0>)
   ) {enabled=false}
   &ApplyInitialForces() {enabled=false}
   /Niagara/Modules/Spawn/Orientation/V2/InitialMeshOrientation(Use Orientation Vector=true, Use Rotation Vector=true) {enabled=false}
  ]
  particleUpdateScript [
   /Game/Effects/Niagara/Modules/Orbit/FN_Orbit_v3(
    Radius min/max:Vector2f=<10 15>
    Rotation Rate min/max:Vector2f=<0.1 0.2>
    To Axis:Vector3f=&Lerp_Vector(A=<-0.25 0 1>, B=<0.25 0.25 1>, Alpha=0.5)
   )
   &SolveForcesAndVelocity(Clamp Velocity=false, Limit Acceleration=false)
   &ParticleState(Kill Particles When Lifetime Has Elapsed=false, Let Infinitely Lived Particles Die When Emitter Deactivates=true)
   &Color()
  ]
  rendererClasses [&NiagaraMeshRendererProperties]
  renderers [{rendererClass=&NiagaraMeshRendererProperties, rendererIndex=0}]
 }}
summary {
 systemName=NS_FlopperTank_02
 emitters [{emitterName=Fish, rendererClasses [&NiagaraMeshRendererProperties]}]
 userVariables []
}
